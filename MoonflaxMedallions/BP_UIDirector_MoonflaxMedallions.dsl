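; UI director for Moonflax Medallions that shows or hides the Moonflax equipped-item UI scene based on a gameplay tag.
; BeginPlay starts the tag wait on the owning player controller and binds view target changes, which restart it on the new target.
; With a valid controller and Athena pawn, async tasks watch the actor, adding the scene on tag add and removing it on tag removal.
; Earlier waits are not explicitly cancelled.
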(event EventBeginPlay
  (bind _asfort_player_controller (Utilities|Casting|CastToFortPlayerController (DynamicUIDirector|GetOwningLocalPlayerController)))
  (EventDispatchers|BindEventtoOnViewTargetChanged _asfort_player_controller (delegate Custom|OnViewTargetChanged))
  (CallFunction|WaitforMoonflaxMedallionTag _asfort_player_controller))

(event Custom|OnViewTargetChanged
  (params (PlayerController :type Fort%20Player%20Controller%20Object%20Reference) (OldViewTarget :type Actor%20Object%20Reference) (NewViewTarget :type Actor%20Object%20Reference))
  (CallFunction|WaitforMoonflaxMedallionTag NewViewTarget))

(event Custom|WaitForMoonflaxMedallionTag
  (params (PlayerActor :type Actor%20Object%20Reference))
  (bind _returnvalue (DynamicUIDirector|GetOwningLocalPlayerController))
  ;[1 1 1] Wait For Appropriate Tag
  (Utilities|IsValid PlayerActor
    (:Is%20Valid
      (Utilities|IsValid _returnvalue
        (:Is%20Valid
          (Utilities|Casting|CastToFortPlayerPawnAthena (Pawn|GetControlledPawn _returnvalue))
          (Ability|Tasks|WaitGameplayTagAddToActor PlayerActor (Variables|Default|GetMoonFlax%20Medallion%20Tag%20Event)
            (:then
              (Ability|Tasks|WaitGameplayTagRemoveFromActor PlayerActor (Variables|Default|GetMoonFlax%20Medallion%20Tag%20Event)
                (:Removed
                  (DynamicUIDirector|RemoveScene "/MoonflaxMedallions/UI/UIScene_MoonflaxEquippedItem.UIScene_MoonflaxEquippedItem"))))
            (:Added
              (DynamicUIDirector|AddScene "/MoonflaxMedallions/UI/UIScene_MoonflaxEquippedItem.UIScene_MoonflaxEquippedItem"))))))))

(fn ConstructionScript)
