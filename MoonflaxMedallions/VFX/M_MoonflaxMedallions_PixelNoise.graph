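# Unlit translucent Niagara sprite material that renders pixelated, animated noise within a soft sphere mask for Moonflax Medallions VFX.
# UVs snap to a PixelSize grid and time is quantized, so two multiplied, vertically counter-scrolling Perlin samples animate in chunky steps.
# The noise blends LightColor to ShadowColor, tinted by particle color through MF_FX_EmissiveAdjustments.
# Noise subtracted from particle alpha erodes opacity, shaped by the MaskOffset-shifted pixelated mask and particle alpha.
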Material /MoonflaxMedallions/VFX/M_MoonflaxMedallions_PixelNoise
graph {
 properties {
  bAutomaticallySetUsageInEditor=false
  bUsedWithNiagaraSprites=true
  blendMode=BLEND_Translucent
  pixelDepthOffsetMode=PDOM_Legacy
  shadingModel=MSM_Unlit
  usedShadingModels=Unlit
 }
 nodes [
  n1 = ParticleColor()
  n2 = VectorParameter() {parameterName=LightColor, defaultValue=<rgba 0 0.7991 1 1>}
  n3 = VectorParameter() {parameterName=ShadowColor, defaultValue=<rgba 0.5075 0.4411 3 1>}
  n4 = TextureCoordinate()
  n5 = ScalarParameter() {defaultValue=18, parameterName=PixelSize}
  n6 = Divide(A=Floor(None=Multiply(A=@n4, B=@n5) {constB=15}), B=@n5) {constB=15}
  n7 = Time()
  n8 = Divide(A=Floor(None=Multiply(A=@n7) {constB=12})) {constB=12}
  n9 = Multiply(A=@n8, B=Constant2Vector() {g=0.3})
  n10 = Multiply(A=@n8, B=Constant2Vector() {g=-0.5})
  n11 = Multiply(
   A=TextureSample(UVs=Add(A=@n6, B=@n9)) {texture=/Game/Global/Textures/Default/Noise/3xPerlin_M}.R
   B=TextureSample(UVs=Add(A=@n6, B=@n10)) {texture=&3xPerlin_M}.B
  )
  n12 = Power(Base=Saturate(None=Multiply(A=@n11) {constB=10})) {constExponent=1}
  MaterialFunctionCall(Emissive=Multiply(A=@n1.RGB, B=LinearInterpolate(A=@n2.RGB, B=@n3.RGB, Alpha=@n12))) {materialFunction=&MF_FX_EmissiveAdjustments}
  n13 = VectorParameter() {parameterName=MaskOffset, defaultValue=<rgba 0 0 0 1>}
  n14 = TextureSample!(UVs=Add(A=@n6, B=ComponentMask(None=@n13.RGB) {g=true, r=true}), Tex, Level, Apply View MipBias) {automaticViewMipBias=false, mipValueMode=TMVM_MipLevel, texture=/Game/Global/Textures/Default/Masks_and_Gradients/WM_PinGlow_Sphere}
  Multiply(A=@n1.A, B=Multiply(A=Subtract(A=@n1.A, B=@n12), B=Power(Base=@n14.R) {constExponent=0.5}))
 ]
}
